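; PinkOatmeal Sepia environmental item that applies a sepia swap gameplay effect to Athena player pawns crossing its volume.
; On BeginPlay it binds handlers to the volume's actor enter and exit events.
; Both handlers call the same MCast-named event, which casts to PlayerPawn_Athena and applies GE_PinkOatmeal_Sepia_Swap to that pawn's ability system.
; Exit applies the effect again and never explicitly removes it.
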(event EventBeginPlay
  (|CallParentFunction)
  (Volume|BindEventtoOnActorEnter (EventDispatchers|CreateEvent :!event Custom|OnActorEnterFX))
  (Volume|BindEventtoOnActorExit (EventDispatchers|CreateEvent :!event Custom|OnACtorExitFX)))

(event Custom|OnActorEnterFX
  (params (Actor :type Actor%20Object%20Reference) (ActorComponent :type Actor%20Component%20Object%20Reference) (InstigatorPawn :type Pawn%20Object%20Reference) (Volume :type Fort%20Volume%20Object%20Reference))
  (CallFunction|OnActorEnterFXMCast Actor))

(event Custom|OnActorEnterFXMCast
  (params (Actor :type Actor%20Object%20Reference))
  (bind _asplayer_pawn_athena (Utilities|Casting|CastToPlayerPawn_Athena Actor))
  (GameplayEffects|ApplyGameplayEffectToSelf :self (Ability|GetAbilitySystemComponent _asplayer_pawn_athena) :GameplayEffectClass "/Figment_S10/Athena/Items/EnvironmentalItems/PinkOatmeal/GE_PinkOatmeal_Sepia_Swap.GE_PinkOatmeal_Sepia_Swap_C" :Level 0.0))

(event Custom|OnACtorExitFX
  (params (Actor :type Actor%20Object%20Reference) (ActorComponent :type Actor%20Component%20Object%20Reference) (InstigatorPawn :type Pawn%20Object%20Reference) (Volume :type Fort%20Volume%20Object%20Reference))
  (CallFunction|OnActorEnterFXMCast Actor))

(fn ConstructionScript
  (|CallParentFunction))
